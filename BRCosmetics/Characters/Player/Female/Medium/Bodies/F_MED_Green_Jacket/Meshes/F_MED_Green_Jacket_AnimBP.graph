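# Post-process-style AnimBlueprint for the female medium Green Jacket cosmetic body, copying the attached parent mesh's pose and adding secondary motion.
# Thigh-to-pelvis rotation alphas drive corrective pelvis and thigh-volume offsets, while constraints position both bag bones and the enlarged neck-scale bone.
# Paired AnimDynamics on the bags switch between tighter front-end and looser in-game limits at half alpha.
# Crouching blends in an extra left-bag offset through a cached pose.
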AnimBlueprint /BRCosmetics/Characters/Player/Female/Medium/Bodies/F_MED_Green_Jacket/Meshes/F_MED_Green_Jacket_AnimBP
target_skeleton=&/F_MED_Green_Jacket_Skeleton
graphs [
 AnimGraph {nodes [
   n1 = "Transform (Modify) Bone - Bone: dyn_l_bag"(ComponentPose=Local To Component(LocalPose=Use cached pose 'PreBag'()), Alpha=1, Translation="-1,0,0", Rotation="10,0,10", Scale="1,1,1") {
    rotationMode=BMM_Additive
    rotationSpace=BCS_BoneSpace
    scaleMode=BMM_Additive
    scaleSpace=BCS_BoneSpace
    translationMode=BMM_Additive
    translationSpace=BCS_BoneSpace
    alphaBoolBlend {blendOption=Linear}
    boneToModify {boneName=dyn_l_bag}
    rotation {pitch=10, roll=10}
    translation {x=-1}
   }
   Output Pose(Result=Blend Poses by bool(BlendPose_0=Component To Local(ComponentPose=@n1), BlendPose_1=Use cached pose 'PreBag'(), bActiveValue=Get bIsCrouching()) {blendType=Linear}) {layerGroup=None}
   n2 = Get thigh_rot_l()
   n3 = "Transform (Modify) Bone - Bone: pelvis_l"(
    ComponentPose=Local To Component(LocalPose=Copy Pose From Mesh() {bUseAttachedParent=true})
    Alpha=@n2
    Translation="-3,-0.93,0"
    Rotation="0,0,0"
    Scale="1,1,1"
   ) {
    rotationSpace=BCS_BoneSpace
    scaleMode=BMM_Additive
    scaleSpace=BCS_BoneSpace
    translationMode=BMM_Additive
    translationSpace=BCS_BoneSpace
    alphaBoolBlend {blendOption=Linear}
    boneToModify {boneName=pelvis_l}
    translation=<-3 -0.93>
   }
   n4 = "Transform (Modify) Bone - Bone: pelvis_mid_l"(ComponentPose=@n3, Alpha=@n2, Translation="-1.461,0.886,0.000", Rotation="0,0,0", Scale="1,1,1") {
    rotationSpace=BCS_BoneSpace
    scaleMode=BMM_Additive
    scaleSpace=BCS_BoneSpace
    translationMode=BMM_Additive
    translationSpace=BCS_BoneSpace
    alphaBoolBlend {blendOption=Linear}
    boneToModify {boneName=pelvis_mid_l}
    translation=<-1.461 0.886>
   }
   n5 = "Transform (Modify) Bone - Bone: pelvis_side_l"(ComponentPose=@n4, Alpha=@n2, Translation="-1.165,0.593,0.240", Rotation="0,0,0", Scale="1,1,1") {
    rotationSpace=BCS_BoneSpace
    scaleMode=BMM_Additive
    scaleSpace=BCS_BoneSpace
    translationMode=BMM_Additive
    translationSpace=BCS_BoneSpace
    alphaBoolBlend {blendOption=Linear}
    boneToModify {boneName=pelvis_side_l}
    translation=<-1.165 0.593 0.24>
   }
   n6 = "Transform (Modify) Bone - Bone: thigh_volume_bk_l"(ComponentPose=@n5, Alpha=@n2, Translation="20,-11,0", Rotation="0,0,0", Scale="1,1,1") {
    rotationSpace=BCS_BoneSpace
    scaleMode=BMM_Additive
    scaleSpace=BCS_BoneSpace
    translationMode=BMM_Additive
    translationSpace=BCS_BoneSpace
    alphaBoolBlend {blendOption=Linear}
    boneToModify {boneName=thigh_volume_bk_l}
    translation=<20 -11>
   }
   n7 = "Transform (Modify) Bone - Bone: thigh_volume_fr_l"(ComponentPose=@n6, Alpha=@n2, Translation="-2.144,7.226,-0.000", Rotation="0,0,0", Scale="1,1,1") {
    rotationSpace=BCS_BoneSpace
    scaleMode=BMM_Additive
    scaleSpace=BCS_BoneSpace
    translationMode=BMM_Additive
    translationSpace=BCS_BoneSpace
    alphaBoolBlend {blendOption=Linear}
    boneToModify {boneName=thigh_volume_fr_l}
    translation=<-2.144 7.226>
   }
   n8 = Get thigh_rot_r()
   n9 = "Transform (Modify) Bone - Bone: pelvis_r"(ComponentPose=@n7, Alpha=@n8, Translation="-3.022,-0.930,0.000", Rotation="0,0,0", Scale="1,1,1") {
    rotationSpace=BCS_BoneSpace
    scaleMode=BMM_Additive
    scaleSpace=BCS_BoneSpace
    translationMode=BMM_Additive
    translationSpace=BCS_BoneSpace
    alphaBoolBlend {blendOption=Linear}
    boneToModify {boneName=pelvis_r}
    translation=<-3.022 -0.93>
   }
   n10 = "Transform (Modify) Bone - Bone: pelvis_mid_r"(ComponentPose=@n9, Alpha=@n8, Translation="-1.461,-1.001,-0.000", Rotation="0,0,0", Scale="1,1,1") {
    rotationSpace=BCS_BoneSpace
    scaleMode=BMM_Additive
    scaleSpace=BCS_BoneSpace
    translationMode=BMM_Additive
    translationSpace=BCS_BoneSpace
    alphaBoolBlend {blendOption=Linear}
    boneToModify {boneName=pelvis_mid_r}
    translation=<-1.461 -1.001>
   }
   n11 = "Transform (Modify) Bone - Bone: pelvis_side_r"(ComponentPose=@n10, Alpha=@n8, Translation="-1.165,0.000,-0.817", Rotation="0,0,0", Scale="1,1,1") {
    rotationSpace=BCS_BoneSpace
    scaleMode=BMM_Additive
    scaleSpace=BCS_BoneSpace
    translationMode=BMM_Additive
    translationSpace=BCS_BoneSpace
    alphaBoolBlend {blendOption=Linear}
    boneToModify {boneName=pelvis_side_r}
    translation {x=-1.165, z=-0.817}
   }
   n12 = "Transform (Modify) Bone - Bone: thigh_volume_bk_r"(ComponentPose=@n11, Alpha=@n8, Translation="2.860,5.043,0.000", Rotation="0,0,0", Scale="1,1,1") {
    rotationSpace=BCS_BoneSpace
    scaleMode=BMM_Additive
    scaleSpace=BCS_BoneSpace
    translationMode=BMM_Additive
    translationSpace=BCS_BoneSpace
    alphaBoolBlend {blendOption=Linear}
    boneToModify {boneName=thigh_volume_bk_r}
    translation=<2.86 5.043>
   }
   n13 = "Transform (Modify) Bone - Bone: thigh_volume_fr_r"(ComponentPose=@n12, Alpha=@n8, Translation="-2.144,-7.226,-0.000", Rotation="0,0,0", Scale="1,1,1") {
    rotationSpace=BCS_BoneSpace
    scaleMode=BMM_Additive
    scaleSpace=BCS_BoneSpace
    translationMode=BMM_Additive
    translationSpace=BCS_BoneSpace
    alphaBoolBlend {blendOption=Linear}
    boneToModify {boneName=thigh_volume_fr_r}
    translation=<-2.144 -7.226>
   }
   n14 = "Constraint - dyn_l_bag (P)"(ComponentPose=@n13, Alpha=1, ConstraintWeights_0=1) {
    alphaBoolBlend {blendOption=Linear}
    boneToModify {boneName=dyn_l_bag}
    constraintSetup [{targetBone {boneName=thigh_twist_01_l}}]
    constraintWeights [1]
   }
   n15 = "Constraint - dyn_r_bag (P)"(ComponentPose=@n14, Alpha=1, ConstraintWeights_0=1) {
    alphaBoolBlend {blendOption=Linear}
    boneToModify {boneName=dyn_r_bag}
    constraintSetup [{targetBone {boneName=thigh_twist_01_r}}]
    constraintWeights [1]
   }
   n16 = "Constraint - dyn_neckScale (PPP)"!(ComponentPose=@n15, Alpha=1, ConstraintWeights_0=0.5, ConstraintWeights_1:float=0.5, ConstraintWeights_2:float=0.4) -> (Pose) {
    alphaBoolBlend {blendOption=Linear}
    boneToModify {boneName=dyn_neckScale}
    constraintSetup [{targetBone {boneName=neck_02}}, {targetBone {boneName=neck_01}}, {targetBone {boneName=head}}]
    constraintWeights [1,1,1]
   }
   n17 = "Transform (Modify) Bone - Bone: dyn_neckScale"(ComponentPose=@n16, Alpha=1, Translation="0,0,0", Rotation="0,0,0", Scale="1.14,1.14,1.14") {
    rotationMode=BMM_Additive
    scaleMode=BMM_Additive
    translationMode=BMM_Additive
    alphaBoolBlend {blendOption=Linear}
    boneToModify {boneName=dyn_neckScale}
    scale=<1.14 1.14 1.14>
   }
   n18 = Get frontend active() -> (frontend active:double)
   n19 = Anim Dynamics - Bone: dyn_l_bag(ComponentPose=@n17, Alpha=@n18, ExternalForce="0,0,0") {
    angularDampingOverride=0.9999
    angularSpringConstant=500
    bAngularSpring=true
    bGravityOverrideInSimSpace=true
    bLinearSpring=true
    bOverrideAngularDamping=true
    bOverrideLinearDamping=true
    gravityScale=0.9
    lODThreshold=1
    linearDampingOverride=0.9999
    linearSpringConstant=500
    simulationSpace=BoneRelative
    alphaBoolBlend {blendOption=Linear}
    boundBone {boneName=dyn_l_bag}
    componentLinearAccScale=<0.25 0.25 0.25>
    gravityOverride {z=-980}
    physicsBodyDefinitions [{
      sphereCollisionRadius=0
      boundBone {boneName=dyn_l_bag}
      boxExtents=<10 10 5>
      constraintSetup {angularTargetAxis=AxisZ, angularLimitsMax=<5 1 1>, angularLimitsMin=<-5 -1 -1>, angularTarget=<0 0 1>}
      localJointOffset=<0 0 -1>
     }]
    relativeSpaceBone {boneName=pelvis}
   }
   n20 = Anim Dynamics - Bone: dyn_r_bag(ComponentPose=@n19, Alpha=@n18, ExternalForce="0,0,0") {
    angularDampingOverride=0.9999
    angularSpringConstant=500
    bAngularSpring=true
    bGravityOverrideInSimSpace=true
    bLinearSpring=true
    bOverrideAngularDamping=true
    bOverrideLinearDamping=true
    gravityScale=0.9
    lODThreshold=1
    linearDampingOverride=0.9999
    linearSpringConstant=500
    simulationSpace=BoneRelative
    alphaBoolBlend {blendOption=Linear}
    boundBone {boneName=dyn_r_bag}
    componentLinearAccScale=<0.25 0.25 0.25>
    gravityOverride {z=-980}
    physicsBodyDefinitions [{
      sphereCollisionRadius=0
      boundBone {boneName=dyn_r_bag}
      boxExtents=<10 10 5>
      constraintSetup {angularTargetAxis=AxisZ, angularLimitsMax=<5 1 1>, angularLimitsMin=<-5 -1 -1>, angularTarget=<0 0 1>}
      localJointOffset=<0 0 -1>
     }]
    relativeSpaceBone {boneName=pelvis}
   }
   n21 = Get frontend passive() -> (frontend passive:double)
   n22 = Anim Dynamics - Bone: dyn_l_bag(ComponentPose=@n20, Alpha=@n21, ExternalForce="0,0,0") {
    angularDampingOverride=0.9999
    angularSpringConstant=500
    bAngularSpring=true
    bGravityOverrideInSimSpace=true
    bLinearSpring=true
    bOverrideAngularDamping=true
    bOverrideLinearDamping=true
    gravityScale=0.9
    lODThreshold=1
    linearDampingOverride=0.9999
    linearSpringConstant=500
    simulationSpace=BoneRelative
    alphaBoolBlend {blendOption=Linear}
    boundBone {boneName=dyn_l_bag}
    componentLinearAccScale=<0.25 0.25 0.25>
    gravityOverride {z=-980}
    physicsBodyDefinitions [{
      sphereCollisionRadius=0
      boundBone {boneName=dyn_l_bag}
      boxExtents=<10 10 5>
      constraintSetup {
       angularTargetAxis=AxisZ
       angularLimitsMax=<20 5 5>
       angularLimitsMin=<-20 -5 -5>
       angularTarget=<0 0 1>
       linearAxesMax=<0 0 0.5>
       linearAxesMin=<0 0 -0.5>
      }
      localJointOffset=<0 0 -2>
     }]
    relativeSpaceBone {boneName=pelvis}
   }
   n23 = Anim Dynamics - Bone: dyn_r_bag(ComponentPose=@n22, Alpha=@n21, ExternalForce="0,0,0") {
    angularDampingOverride=0.9999
    bGravityOverrideInSimSpace=true
    bOverrideAngularDamping=true
    bOverrideLinearDamping=true
    gravityScale=0.9
    lODThreshold=1
    linearDampingOverride=0.9999
    simulationSpace=BoneRelative
    alphaBoolBlend {blendOption=Linear}
    boundBone {boneName=dyn_r_bag}
    componentLinearAccScale=<0.25 0.25 0.25>
    gravityOverride {z=-980}
    physicsBodyDefinitions [{
      sphereCollisionRadius=0
      boundBone {boneName=dyn_r_bag}
      boxExtents=<10 10 5>
      constraintSetup {
       angularTargetAxis=AxisZ
       angularLimitsMax=<20 5 5>
       angularLimitsMin=<-20 -5 -5>
       angularTarget=<0 0 1>
       linearAxesMax=<0 0 0.5>
       linearAxesMin=<0 0 -0.5>
      }
      localJointOffset=<0 0 -2>
     }]
    relativeSpaceBone {boneName=pelvis}
   }
   Save cached pose 'PreBag'(Pose=Component To Local(ComponentPose=@n23)) {cacheName=PreBag}
  ]}
 EventGraph {nodes [
   Event OnCharacterCustomizationLoaded()
   > Do Once()
   n1 = CheckForBackbling()
   > Set isWearingBackbling(isWearingBackbling=OR Boolean(A=@n1, B=@n1.bIsBackblingLong))
   Event BlueprintInitializeAnimation()
   > Set Skel_Mesh(Skel_Mesh=GetOwnerSkeletalMeshComponent())
   Event BlueprintUpdateAnimation()
   n2 = Get Skel_Mesh()
   > Set thigh_rot_r(thigh_rot_r=GetAlphaFromReferenceTransformDelta(SkeletalMeshComponent=@n2, BoneName=thigh_r, BaseName=pelvis, InputRange=(X=0,Y=-90), TwistAxis="0,0,1"))
   > Set thigh_rot_l(thigh_rot_l=GetAlphaFromReferenceTransformDelta(SkeletalMeshComponent=@n2, BoneName=thigh_l, BaseName=pelvis, InputRange=(X=-15,Y=-110), TwistAxis="0,0,1"))
   > Set frontend active(frontend active:double=SelectFloat(A=0.5, B=0, bPickA=Get bIsInFrontEnd())) -> (Output_Get:double)
   > Set frontend passive(frontend passive:double=SelectFloat(A=0, B=0.5, bPickA=Get bIsInFrontEnd())) -> (Output_Get:double)
  ]}
]
variables [
 frontend active {type=number}
 frontend passive {type=number}
 isWearingBackbling {type=boolean}
 Skel_Mesh {type=SkeletalMeshComponent}
 thigh_rot_l {type=number}
 thigh_rot_r {type=number}
]
